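# Niagara swing trail for the Gourd Riddance pickaxe, layering ribbons, glow sprites, embers and leaves.
# Two anim-trail ribbons, olive and orange, follow bone/socket pairs on the weapon mesh, with time-of-day alpha.
# Despite their names, the Debug emitters draw cyan glows at two sockets: sprites, plus curved ribbons lagging via per-particle position interpolation, tinted at night.
# GPU embers spawn by distance around socket samples, while burst leaves tumble with drag, gravity and curl noise.
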NiagaraSystem /BRCosmetics/Weapons/FORT_Melee/Pickaxe_GourdRiddance/FX/NS_Pickaxe_GourdRiddance_Trail
dependencies {
 usedDataInterfaces [
  &NiagaraDataInterfaceColorCurve
  &NiagaraDataInterfaceCurve
  &NiagaraDataInterfaceArrayFloat
  &NiagaraDataInterfaceSpriteRendererInfo
  &NiagaraDataInterfaceSkeletalMesh
 ]
 usedDynamicInputs []
 usedModules [
  &self:SystemSpawnScript.SystemScriptSource.SystemScriptGraph.NiagaraNodeAssignment_2.SetVariables_#1
  &self:SystemSpawnScript.SystemScriptSource.SystemScriptGraph.NiagaraNodeAssignment_0.SetVariables_#2
  &SystemState
 ]
}
emitters {
 Debug {
  emitterUpdateScript [
   &EmitterState(Life Cycle Mode=::NewEnumerator1, Loop Behavior=::NewEnumerator1, Loop Duration=0.5, UseLoopDelay=false)
   &SpawnBurst_Instantaneous(Spawn Count=30, Use Spawn Probability=false, UseLoopCountLimit=false)
  ]
  particleSpawnScript [
   &PartitionParticles(
    Normalized Index In Partition Includes One=true
    Particles Per Partition=15
    Partition Mode:ENiagara_PartitionMode=::NewEnumerator1
    Write Partition Index to Ribbon Link Order:ENiagara_AttributeSamplingApplyOutput=::NewEnumerator0
    Write Partition to Ribbon ID=::NewEnumerator0
   )
   &InitializeParticle(Color=<rgba 1 1 1 2>, Lifetime=0.3, Ribbon Width=10, Ribbon Width Mode=::NewEnumerator1, UsePositionOffset=false, Write Material Random=true)
   &DynamicMaterialParameters(
    Index 0 Param 1=&RandomRangeFloat(Minimum=0, Maximum=1)
    Index 0 Param 2=&RandomRangeFloat(Minimum=0, Maximum=1)
    Module.Param0WriteEnabled=true
    Module.Param1WriteEnabled=true
    Module.Param2WriteEnabled=false
    Module.Param3WriteEnabled=false
   )
   &CameraOffset(Camera Offset Amount=10)
  ]
  particleUpdateScript [
   &ParticleState()
   &SkeletalMeshLocation(
    Bone / Socket Index=&Add_Integer(A=$Particles.PartitionParticles.Partition, B=3)
    MaskSpawnGroup=false
    Mesh Sampling Type=::NewEnumerator5
    Sampled Position Offset=&MakeVector(
     X=&FloatFromCurve(
      FloatCurve=DataInterface() {propertyValues {Curve {keys [
          {interpMode=RCIM_Cubic, tangentMode=RCTM_User, time=0, value=0}
          {arriveTangent=3.0667, interpMode=RCIM_Cubic, leaveTangent=3.0667, tangentMode=RCTM_User, time=1, value=1}
         ]}}}
      CurveIndex=$Particles.PartitionParticles.Normalized Index In Partition
      Scale Curve=-10
     )
     Y=&FloatFromCurve(
      FloatCurve=DataInterface() {propertyValues {Curve {keys [
          {interpMode=RCIM_Cubic, tangentMode=RCTM_User, time=0, value=0}
          {arriveTangent=1.5478, interpMode=RCIM_Cubic, leaveTangent=1.5478, tangentMode=RCTM_User, time=1, value=1}
         ]}}}
      CurveIndex=$Particles.PartitionParticles.Normalized Index In Partition
      Scale Curve=10
     )
     Z=&FloatFromCurve(
      FloatCurve=DataInterface() {propertyValues {Curve {keys [
          {interpMode=RCIM_Cubic, tangentMode=RCTM_User, time=0, value=0}
          {arriveTangent=3.0667, interpMode=RCIM_Cubic, leaveTangent=3.0667, tangentMode=RCTM_User, time=1, value=1}
         ]}}}
      CurveIndex=$Particles.PartitionParticles.Normalized Index In Partition
      Scale Curve=&SelectFloatFromArray(Direct Array Index=$Particles.PartitionParticles.Partition, Float Selection Array=DataInterface() {propertyValues {FloatData [10,-10]}})
     )
    )
    Skeletal Mesh=$System.Skeletal Mesh
    Socket Sampling Mode=::NewEnumerator5
    Use Sampled Mesh Attribute as Ribbon ID=false
   )
   /SetVariables(Particles.Position_Glow:NiagaraPosition=$Output.SkeletalMeshLocation.SampledPositionWithoutOffset) {isSetParameters=true}
   &InterpolateOverTime(
    Rate Of Change:NiagaraFloat=&Lerp_Float(
     A=0.5
     B=50
     Alpha=&FloatFromCurve(
      FloatCurve=DataInterface() {propertyValues {Curve {keys [
          <at 0 1 RCIM_Cubic>
          {arriveTangent=-0.4126, interpMode=RCIM_Cubic, leaveTangent=-0.4126, tangentMode=RCTM_User, time=0.25, value=1}
          {arriveTangent=0.0584, interpMode=RCIM_Cubic, leaveTangent=0.0584, tangentMode=RCTM_User, time=1, value=0.1}
         ]}}}
      CurveIndex=$Particles.PartitionParticles.Normalized Index In Partition
     )
    )
    Target Position Value:NiagaraPosition=$Particles.Position
    Variable Type=::NewEnumerator6
   )
   /SetVariables(Particles.Position=$Particles.InterpolateOverTime.InterpolateOverTime_Position.Moving Average) {isSetParameters=true}
   &DynamicMaterialParameters(
    Index 0 Param 3=&FloatFromCurve(FloatCurve=DataInterface() {propertyValues {Curve {keys [
         {arriveTangent=-1.8694, interpMode=RCIM_Cubic, leaveTangent=-1.8694, tangentMode=RCTM_User, time=0, value=1}
         {interpMode=RCIM_Cubic, tangentMode=RCTM_User, time=1, value=0}
        ]}}})
    Module.Index0Param0Write=false
    Module.Index0Param1Write=false
    Module.Index0Param2Write=true
    Module.Index0Param3Write=false
   ) {module=DynamicMaterialParameters001}
   &ScaleColor(
    Linear Color Curve=DataInterface() {propertyValues {
      AlphaCurve {keys [<at 0 1>, <at 0.6527 0.5>, {arriveTangentWeight=0.326, interpMode=RCIM_Cubic, tangentMode=RCTM_User, time=0.9736, value=0}]}
      BlueCurve {keys [<at 0.1134 2 RCIM_Cubic>, <at 0.4158 0.2984>, {interpMode=RCIM_Cubic, leaveTangent=-5.8066, time=0.9151, value=0.0703}]}
      GreenCurve {keys [
        {arriveTangent=-2.6767, interpMode=RCIM_Cubic, time=0.1134, value=1.8463}
        <at 0.4158 0.4583>
        {interpMode=RCIM_Cubic, leaveTangent=-4.0434, time=0.9151, value=0.0885}
       ]}
      RedCurve {keys [
        {arriveTangent=-6.9405, interpMode=RCIM_Cubic, time=0.1134, value=0}
        <at 0.4158 0.0695>
        {interpMode=RCIM_Cubic, leaveTangent=0.0074, time=0.9151, value=0}
       ]}
     }}
    Scale Mode=::NewEnumerator2
   )
   &ScaleColor(
    Scale Alpha=&FloatFromCurve(
     FloatCurve=DataInterface() {propertyValues {Curve {keys [
         {arriveTangent=-0.3092, interpMode=RCIM_Cubic, leaveTangent=-0.3092, tangentMode=RCTM_User, time=0, value=0}
         {arriveTangent=13.1758, interpMode=RCIM_Cubic, leaveTangent=13.1758, tangentMode=RCTM_User, time=0.35, value=2}
        ]}}}
     CurveIndex=$Emitter.Age
     Scale Curve=1.5
    )
    Scale RGB=&MakeVectorFromLinearColorRGB(Color=&FN_Select_Color_from_Time_Of_Day(Morning=<rgba 1 1 1 1>, Day=<rgba 1 1 1 1>, Evening=<rgba 1 1 1 1>, Night=<rgba 0.1097 0.9938 2 1>))
    ScaleA=true
    ScaleRGB=true
   ) {module=ScaleColor001}
  ]
  rendererClasses [&NiagaraRibbonRendererProperties]
  renderers [{rendererClass=&NiagaraRibbonRendererProperties}]
 }
 Debug001 {
  emitterUpdateScript [
   &EmitterState(Life Cycle Mode=::NewEnumerator1, Loop Behavior=::NewEnumerator1, Loop Duration=0.5, UseLoopDelay=false)
   &SpawnBurst_Instantaneous(Spawn Count=2, Use Spawn Probability=false, UseLoopCountLimit=false)
  ]
  particleSpawnScript [
   &PartitionParticles(Write Partition to Ribbon ID=::NewEnumerator0)
   &InitializeParticle(Lifetime=0.5, Sprite Size Mode=::NewEnumerator3, Uniform Sprite Size=5, UsePositionOffset=false, Write Material Random=true)
   &CameraOffset(Camera Offset Amount=5)
  ]
  particleUpdateScript [
   &ParticleState()
   &SkeletalMeshLocation(
    Bone / Socket Index=&Add_Integer(A=$Particles.PartitionParticles.Partition, B=3)
    MaskSpawnGroup=false
    Mesh Sampling Type=::NewEnumerator5
    Skeletal Mesh=$System.Skeletal Mesh
    Socket Sampling Mode=::NewEnumerator5
    Use Sampled Mesh Attribute as Ribbon ID=false
   )
   &ScaleColor(
    Linear Color Curve=DataInterface() {propertyValues {
      AlphaCurve {keys [<at 0 0.5>, {arriveTangentWeight=0.326, interpMode=RCIM_Cubic, tangentMode=RCTM_User, time=1, value=0}]}
      BlueCurve {keys [<at 0 2 RCIM_Cubic>, <at 0.2538 0.2984>, {interpMode=RCIM_Cubic, leaveTangent=-5.8066, time=0.5152, value=0.0703}]}
      GreenCurve {keys [
        {arriveTangent=-2.6767, interpMode=RCIM_Cubic, time=0, value=1.8463}
        <at 0.2538 0.4583>
        {interpMode=RCIM_Cubic, leaveTangent=-4.0434, time=0.5152, value=0.0885}
       ]}
      RedCurve {keys [
        {arriveTangent=-6.9405, interpMode=RCIM_Cubic, time=0, value=0}
        <at 0.2538 0.0695>
        {interpMode=RCIM_Cubic, leaveTangent=0.0074, time=0.5152, value=0}
       ]}
     }}
    Scale Mode=::NewEnumerator2
   )
   &ScaleColor(
    Scale Alpha=&FloatFromCurve(
     FloatCurve=DataInterface() {propertyValues {Curve {keys [
         {arriveTangent=-0.3092, interpMode=RCIM_Cubic, leaveTangent=-0.3092, tangentMode=RCTM_User, time=0, value=0}
         {arriveTangent=13.1758, interpMode=RCIM_Cubic, leaveTangent=13.1758, tangentMode=RCTM_User, time=0.35, value=2}
        ]}}}
     CurveIndex=$Emitter.Age
     Scale Curve=2
    )
    ScaleA=true
    ScaleRGB=false
   ) {module=ScaleColor001}
  ]
  rendererClasses [&NiagaraSpriteRendererProperties]
  renderers [{rendererClass=&NiagaraSpriteRendererProperties}]
 }
 Embers {
  simTarget=GPUComputeSim
  emitterUpdateScript [
   &EmitterState()
   &SpawnPerUnit(Module.Use Max Movement Threshold=true, Module.Use Movement Tolerance=false, Use Spawn Probability=false)
  ]
  particleSpawnScript [
   &InitializeParticle(
    Color Mode=::NewEnumerator0
    Lifetime Max=0.2
    Lifetime Min=0.1
    Lifetime Mode=::NewEnumerator1
    Sprite Size Max=<2.5 2>
    Sprite Size Min=<4 3>
    Sprite Size Mode=::NewEnumerator2
    UsePositionOffset=false
    Write Material Random=true
   )
   &PartitionParticles(Write Partition to Ribbon ID=::NewEnumerator0)
   &SkeletalMeshLocation(
    Bone / Socket Index=&Add_Integer(A=$Particles.PartitionParticles.Partition, B=3)
    MaskSpawnGroup=false
    Mesh Sampling Type=::NewEnumerator5
    Skeletal Mesh=$System.Skeletal Mesh
    Socket Sampling Mode=::NewEnumerator5
    Use Sampled Mesh Attribute as Ribbon ID=false
   )
   &ShapeLocation(
    MaskSpawnGroup=false
    Offset Coordinate Space=::Local
    Offset Mode=::NewEnumerator0
    Override Seed=false
    Shape Origin=$Particles.SkeletalMeshLocation.SampledPosition
    Sphere Radius=&RandomRangeFloat(Minimum=5, Maximum=10)
    UseOwnerScale=true
   )
   &InheritVelocity(Speed Limit=true, UseSpeedThreshold=false)
   &ApplyInitialForces()
  ]
  particleUpdateScript [
   &Drag(Drag=&RandomRangeFloat(Minimum=20, Maximum=40), Use Linear Drag=true, Use Rotational Drag=true)
   &SolveForcesAndVelocity(Clamp Velocity=false, Limit Acceleration=false)
   &ParticleState()
   &ScaleSpriteSize(
    Uniform Curve Scale=0.75
    Uniform Curve Sprite Scale=DataInterface() {propertyValues {Curve {keys [
        {interpMode=RCIM_Cubic, tangentMode=RCTM_User, time=0, value=0}
        <at 0.2 1>
        {interpMode=RCIM_Cubic, tangentMode=RCTM_User, time=0.7, value=1}
        {tangentMode=RCTM_User, time=1, value=0}
       ]}}}
   )
   &ScaleSpriteSizeBySpeed(Max Scale Factor=<0.5 1.8>, Min Scale Factor=<0.5 0.9>, Velocity Threshold=100)
   &ScaleColor(
    Scale RGB=&MakeVectorFromLinearColorRGB(Color=&ColorFromCurve(ColorCurve=DataInterface() {propertyValues {
        AlphaCurve {keys [<at 0 1>, {arriveTangentWeight=0.326, interpMode=RCIM_Cubic, tangentMode=RCTM_User, time=1, value=0}]}
        BlueCurve {keys [<at 0 2 RCIM_Cubic>, <at 0.2538 0.2984>, {interpMode=RCIM_Cubic, leaveTangent=-5.8066, time=0.5152, value=0.0703}]}
        GreenCurve {keys [
          {arriveTangent=-2.6767, interpMode=RCIM_Cubic, time=0, value=1.8463}
          <at 0.2538 0.4583>
          {interpMode=RCIM_Cubic, leaveTangent=-4.0434, time=0.5152, value=0.0885}
         ]}
        RedCurve {keys [
          {arriveTangent=-6.9405, interpMode=RCIM_Cubic, time=0, value=0}
          <at 0.2538 0.0695>
          {interpMode=RCIM_Cubic, leaveTangent=0.0074, time=0.5152, value=0}
         ]}
       }}))
    ScaleA=true
    ScaleRGB=true
   )
  ]
  rendererClasses [&NiagaraSpriteRendererProperties]
  renderers [{rendererClass=&NiagaraSpriteRendererProperties}]
 }
 Leaves {
  emitterUpdateScript [
   &EmitterState(Life Cycle Mode=::NewEnumerator1, Loop Behavior=::NewEnumerator1, Loop Duration=0.5, UseLoopDelay=false)
   &SpawnBurst_Instantaneous(
    Spawn Count=&RandomRangeInt(Minimum=1, Maximum=2)
    Spawn Probability=0.25
    Use Spawn Probability=false
    UseLoopCountLimit=false
   ) {module=SpawnBurst_Instantaneous001}
  ]
  particleSpawnScript [
   &InitializeParticle(
    Color=&RandomRangeLinearColor(Minimum=<rgba 0.2771 0.3229 0.0323 1>, Maximum=<rgba 0.1388 0.161 0.0139 1>)
    Lifetime Max=0.8
    Lifetime Min=0.6
    Lifetime Mode=::NewEnumerator1
    Sprite Rotation Mode=::NewEnumerator2
    Sprite Size Max=<12.5 7.5>
    Sprite Size Min=<7.5 2.5>
    Sprite Size Mode=::NewEnumerator2
    UsePositionOffset=false
    Write Material Random=true
   ) {module=InitializeParticle001}
   &CylinderLocation(
    Cylinder Height=0
    Cylinder Radius=1
    MaskSpawnGroup=false
    Offset=<0 0 50>
    Offset Coordinate Space:ENiagaraCoordinateSpace=::Local
    Offset Mode:ENiagara_OffsetMode=::NewEnumerator0
    Override Seed=false
    Rotation Mode:ENiagara_RotationMode=::NewEnumerator1
    Surface Only=true
    Use Endcaps In Surface Only Mode=false
    UseOwnerScale=true
   )
   &AddVelocityFromPoint(
    Module.Use Falloff Distance=false
    Origin Offset=<-1 0 0>
    Velocity Strength=&RandomRangeFloat(Minimum=150, Maximum=450)
   ) {module=AddVelocityFromPoint001}
  ]
  particleUpdateScript [
   &ParticleState()
   &ScaleSpriteSize(Uniform Curve Sprite Scale=DataInterface() {propertyValues {Curve {keys [{interpMode=RCIM_Cubic, tangentMode=RCTM_User, time=0, value=1}, {interpMode=RCIM_Cubic, tangentMode=RCTM_User, time=1, value=0}]}}})
   &SpriteRotationRate(Rotation Rate=&FloatFromCurve(
     FloatCurve=DataInterface() {propertyValues {Curve {keys [
         {interpMode=RCIM_Cubic, tangentMode=RCTM_User, time=0, value=1}
         {interpMode=RCIM_Cubic, leaveTangent=-4.1325, tangentMode=RCTM_Break, time=0.5, value=1}
         {interpMode=RCIM_Cubic, tangentMode=RCTM_User, time=1, value=0}
        ]}}}
     Scale Curve=&RandomRangeFloat(Minimum=180, Maximum=360)
    ))
   &SubUVAnimation(
    "Sprite Renderer "=DataInterface() {propertyValues {SpriteRenderer=&self:Leaves_4.NiagaraSpriteRendererProperties_2}}
    SubUV Animation Mode:ENiagara_SubUVLookupModeV2=::NewEnumerator1
    UseEndFrame=true
    UseStartFrame=true
   ) {module=SubUVAnimation002}
   &Drag(Drag=&UniformRangedFloat(Minimum=10, Maximum=15), Use Linear Drag=true, Use Rotational Drag=true) {module=Drag001}
   &AccelerationForce(Acceleration=<0 0 -333>)
   &DynamicMaterialParameters(
    Index 0 Param 1=0.1
    Module.Param0WriteEnabled=true
    Module.Param1WriteEnabled=true
    Module.Param2WriteEnabled=true
    Module.Param3WriteEnabled=true
   ) {module=DynamicMaterialParameters001}
   /Niagara/Modules/Update/Forces/CurlNoiseForce(
    Noise Frequency=0.1
    Noise Quality / Cost=::NewEnumerator2
    Noise Strength=&Multiply_Float(
     A=500
     B=&FloatFromCurve(FloatCurve=DataInterface() {propertyValues {Curve {keys [
          {tangentMode=RCTM_User, time=0, value=1}
          {arriveTangent=-0.0405, interpMode=RCIM_Cubic, leaveTangent=-0.0405, tangentMode=RCTM_User, time=0.7009, value=1}
          {arriveTangent=-8.0089, interpMode=RCIM_Cubic, leaveTangent=-8.0089, tangentMode=RCTM_User, time=1, value=0}
         ]}}})
    )
    Pan Noise=false
   )
   &SolveForcesAndVelocity(Clamp Velocity=false, Limit Acceleration=false)
  ]
  rendererClasses [&NiagaraSpriteRendererProperties]
  renderers [{rendererClass=&NiagaraSpriteRendererProperties}]
 }
 Ribbon_SideVector {
  emitterUpdateScript [
   &EmitterState(Life Cycle Mode=::NewEnumerator1, Loop Behavior=::NewEnumerator1, Loop Duration=0.2, UseLoopDelay=false)
   &FN_SpawnAnimTrail()
  ]
  particleSpawnScript [
   &FN_InitializeAnimTrailBoneIndecies(First Bone or Socket Index=1, Second Bone or Socket Index=0)
   &FN_InitializeAnimTrail(AnimTrailWidth=1.25, Color=<rgba 0.5 0.5 0.5 3>, Lifetime=0.15, SkeletalMesh=$System.Skeletal Mesh)
   &CameraOffset(Camera Offset Amount=5)
  ]
  particleUpdateScript [
   &FN_UpdateAnimTrail(SkeletalMesh=$System.Skeletal Mesh)
   &ScaleColor(
    Linear Color Curve=DataInterface() {propertyValues {
      AlphaCurve {keys [<at 0 2 RCIM_Cubic>,<at 0.1321 1>,<at 0.6162 0 RCIM_Cubic>]}
      BlueCurve {keys [{arriveTangent=0.1336, interpMode=RCIM_Cubic, time=0.009, value=0}, <at 0.0325 0.0344>, <at 0.1349 0.0097>, <at 0.9108 0.0004>]}
      GreenCurve {keys [{arriveTangent=-0.4141, interpMode=RCIM_Cubic, time=0.009, value=0.1927}, <at 0.0325 0.1823>, <at 0.1349 0.0938>, <at 0.9108 0.0208>]}
      RedCurve {keys [{arriveTangent=-16.267, interpMode=RCIM_Cubic, time=0.009, value=0.1674}, <at 0.0325 0.1681>, <at 0.1349 0.0764>, <at 0.9108 0.0121>]}
     }}
    Scale Mode=::NewEnumerator2
   )
   &ScaleColor(
    Scale Alpha=&FN_Select_Float_from_Time_Of_Day(Morning=0.9, Day=0.9, Evening=0.9, Night=0.75)
    ScaleA=true
    ScaleRGB=false
   ) {module=ScaleColor001}
  ]
  rendererClasses [&NiagaraRibbonRendererProperties]
  renderers [{rendererClass=&NiagaraRibbonRendererProperties}]
 }
 Ribbon_SideVector001 {
  emitterUpdateScript [
   &EmitterState(Life Cycle Mode=::NewEnumerator1, Loop Behavior=::NewEnumerator1, Loop Duration=0.2, UseLoopDelay=false)
   &FN_SpawnAnimTrail()
  ]
  particleSpawnScript [
   &FN_InitializeAnimTrailBoneIndecies(First Bone or Socket Index=6, Second Bone or Socket Index=5)
   &FN_InitializeAnimTrail(AnimTrailWidth=1.25, Color=<rgba 0.5 0.5 0.5 2>, Lifetime=0.15, SkeletalMesh=$System.Skeletal Mesh)
   &CameraOffset(Camera Offset Amount=5)
  ]
  particleUpdateScript [
   &FN_UpdateAnimTrail(SkeletalMesh=$System.Skeletal Mesh)
   &ScaleColor(
    Linear Color Curve=DataInterface() {propertyValues {
      AlphaCurve {keys [
        <at 0.0007 2.5>
        {arriveTangent=-3.2415, interpMode=RCIM_Cubic, leaveTangent=-3.2415, tangentMode=RCTM_User, time=0.0657, value=1}
        <at 0.3562 0.75>
        {arriveTangentWeight=0.326, interpMode=RCIM_Cubic, tangentMode=RCTM_User, time=0.9717, value=0}
       ]}
      BlueCurve {keys [<at 0 0.0015 RCIM_Cubic>,<at 0.1293 0.0427>,<at 0.3575 0>,<at 0.6245 0.0008>]}
      GreenCurve {keys [{arriveTangent=-2.6767, interpMode=RCIM_Cubic, time=0, value=0.2844}, <at 0.1293 0.1844>, <at 0.3575 0.069>, <at 0.6245 0.013>]}
      RedCurve {keys [{arriveTangent=-6.9405, interpMode=RCIM_Cubic, time=0, value=1}, <at 0.1293 0.6094>, <at 0.3575 0.276>, <at 0.6245 0.1354>]}
     }}
    Scale Mode=::NewEnumerator2
   )
   &ScaleColor(
    Scale Alpha=&FN_Select_Float_from_Time_Of_Day(Morning=0.75, Day=0.9, Evening=0.75, Night=0.5)
    ScaleA=true
    ScaleRGB=false
   ) {module=ScaleColor001}
  ]
  rendererClasses [&NiagaraRibbonRendererProperties]
  renderers [{rendererClass=&NiagaraRibbonRendererProperties}]
 }
}
summary {
 systemName=NS_Pickaxe_GourdRiddance_Trail
 emitters [
  {emitterName=Ribbon_SideVector, rendererClasses [&NiagaraRibbonRendererProperties]}
  {emitterName=Ribbon_SideVector001, rendererClasses [&NiagaraRibbonRendererProperties]}
  {emitterName=Debug, rendererClasses [&NiagaraRibbonRendererProperties]}
  {emitterName=Debug001, rendererClasses [&NiagaraSpriteRendererProperties]}
  {emitterName=Embers, simTarget=GPUComputeSim, rendererClasses [&NiagaraSpriteRendererProperties]}
  {emitterName=Leaves, rendererClasses [&NiagaraSpriteRendererProperties]}
 ]
 userVariables [
  User.bIsFrontEnd {type=NiagaraBool}
  User.bIsFrontEndPreview {type=NiagaraBool}
  User.RibbonIDOffset {type=NiagaraInt32}
 ]
}
