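# Physics asset for the Gourd Riddance pickaxe mesh, defining simulated secondary motion for its left and right horn bone chains.
# Each side has a small base capsule followed by four identical capsule bodies along the horn segments.
# Constraints link each segment to the previous one, allowing only slight swing and zero twist, so the horns flex subtly as stiff chains.
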PhysicsAsset /BRCosmetics/Weapons/FORT_Melee/Pickaxe_GourdRiddance/Meshes/GourdRiddance_Axe_Physics
bodies {
 dyn_l_horn_1 [{shapeType=Capsule, length=5, radius=3, shapeName=dyn_l_horn_1_capsule, center=<5 0 0>, rotation=<rot -90 0 0>}]
 dyn_l_horn_2 [{shapeType=Capsule, length=5, radius=3, shapeName=dyn_l_horn_2_capsule, center=<5 0 0>, rotation=<rot -90 0 0>}]
 dyn_l_horn_3 [{shapeType=Capsule, length=5, radius=3, shapeName=dyn_l_horn_3_capsule, center=<5 0 0>, rotation=<rot -90 0 0>}]
 dyn_l_horn_4 [{shapeType=Capsule, length=5, radius=3, shapeName=dyn_l_horn_4_capsule, center=<5 0 0>, rotation=<rot -90 0 0>}]
 dyn_l_horn_base [{shapeType=Capsule, length=0.505, radius=0.505, shapeName=dyn_l_horn_base_capsule, rotation=<rot 0 -90 0>}]
 dyn_r_horn_1 [{shapeType=Capsule, length=5, radius=3, shapeName=dyn_r_horn_1_capsule, center=<5 0 0>, rotation=<rot -90 0 0>}]
 dyn_r_horn_2 [{shapeType=Capsule, length=5, radius=3, shapeName=dyn_r_horn_2_capsule, center=<5 0 0>, rotation=<rot -90 0 0>}]
 dyn_r_horn_3 [{shapeType=Capsule, length=5, radius=3, shapeName=dyn_r_horn_3_capsule, center=<5 0 0>, rotation=<rot -90 0 0>}]
 dyn_r_horn_4 [{shapeType=Capsule, length=5, radius=3, shapeName=dyn_r_horn_4_capsule, center=<5 0 0>, rotation=<rot -90 0 0>}]
 dyn_r_horn_base [{shapeType=Capsule, length=0.505, radius=0.505, shapeName=dyn_r_horn_base_capsule, rotation=<rot 0 -90 0>}]
}
constraints [
 {bone1Name=dyn_l_horn_1, bone2Name=dyn_l_horn_base, swing1LimitDegrees=5, swing2LimitDegrees=5, twistLimitDegrees=0}
 {bone1Name=dyn_l_horn_2, bone2Name=dyn_l_horn_1, swing1LimitDegrees=5, swing2LimitDegrees=5, twistLimitDegrees=0}
 {bone1Name=dyn_l_horn_3, bone2Name=dyn_l_horn_2, swing1LimitDegrees=5, swing2LimitDegrees=5, twistLimitDegrees=0}
 {bone1Name=dyn_l_horn_4, bone2Name=dyn_l_horn_3, swing1LimitDegrees=5, swing2LimitDegrees=5, twistLimitDegrees=0}
 {bone1Name=dyn_r_horn_1, bone2Name=dyn_r_horn_base, swing1LimitDegrees=5, swing2LimitDegrees=5, twistLimitDegrees=0}
 {bone1Name=dyn_r_horn_2, bone2Name=dyn_r_horn_1, swing1LimitDegrees=5, swing2LimitDegrees=5, twistLimitDegrees=0}
 {bone1Name=dyn_r_horn_3, bone2Name=dyn_r_horn_2, swing1LimitDegrees=5, swing2LimitDegrees=5, twistLimitDegrees=0}
 {bone1Name=dyn_r_horn_4, bone2Name=dyn_r_horn_3, swing1LimitDegrees=5, swing2LimitDegrees=5, twistLimitDegrees=0}
]
