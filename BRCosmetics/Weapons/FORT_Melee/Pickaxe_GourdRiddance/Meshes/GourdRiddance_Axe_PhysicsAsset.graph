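# Physics asset for the Gourd Riddance two-handed pickaxe cosmetic mesh, providing its simplified collision shape.
# It has a single body on the Melee_TwoHanded_Handle bone, a capsule offset along the handle and rotated to approximate the weapon.
# There are no constraints, so the weapon has no articulated or jointed physics parts.
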PhysicsAsset /BRCosmetics/Weapons/FORT_Melee/Pickaxe_GourdRiddance/Meshes/GourdRiddance_Axe_PhysicsAsset
bodies {Melee_TwoHanded_Handle [{shapeType=Capsule, length=113.0788, radius=22.3633, center=<5.3014 -1.2123 60.0888>, rotation=<rot -0.3862 -90 0.0341>}]}
constraints []
